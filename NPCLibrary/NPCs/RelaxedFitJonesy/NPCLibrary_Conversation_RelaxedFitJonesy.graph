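# Conversation database for the Relaxed Fit Jonesy NPC, defining the greeting shown when a player talks to him.
# A single entry point feeds a speech node with a default line and contextual overrides.
# Overrides cover a first-ever meeting, prioritized by a curve row, and reactions to specific player skins via character ID requirements.
# The speech node then branches through link nodes into the shared NPC Library quests and basic services conversations.
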ConversationDatabase /NPCLibrary/NPCs/RelaxedFitJonesy/NPCLibrary_Conversation_RelaxedFitJonesy
entry_points [{tag=Conversation.Entry.RelaxedFitJonesy, destinations [#1]}]
nodes [
 ConversationEntryPointNode_7 {
  id=#1
  class=ConversationEntryPointNode
  outputs [#2]
  properties {entryTag {tagName=Conversation.Entry.RelaxedFitJonesy}}
 }
 Speech_C_4 {
  id=#2
  class=Speech_C
  outputs [#3, #4]
  properties {
   nodeName=""
   generalConfig {
    defaultMessage=My relaxation is hardcore. You ever knitted a sweater while learning to juggle?
    contextualMessages [
     {
      message="Welcome to Fort Jonesy. You gotta abide by the rules of cool. Rule 1: CHILL!"
      contextRequirements [{requirement {class=FortControllerRequirement_IsFirstEverConversationWithNPC}}]
      priority {curve {rowName=Tandem.Speech.Priority.FirstEverInteraction}}
     }
     {
      message=Take it easy.
      requirementMatchPolicy=RequireAny
      contextRequirements [{requirement {
         class=FortControllerRequirement_HasCID
         properties {allowedCIds [
           /BRCosmetics/Athena/Items/Cosmetics/Characters/CID_409_Athena_Commando_M_BunkerMan
           &CID_883_Athena_Commando_M_ChOneJonesy
           &CID_989_Athena_Commando_M_ProgressiveJonesy
          ]}
        }}]
      priority {value=0, curve {curveTable=null, rowName=None}}
     }
     {
      message="I guess you can't play water polo solo, huh."
      contextRequirements [{requirement {
         class=FortControllerRequirement_HasCID
         properties {allowedCIds [/BRCosmetics/Athena/Items/Cosmetics/Characters/CID_836_Athena_Commando_M_JonesyFlare]}
        }}]
      priority {value=0, curve {curveTable=null, rowName=None}}
     }
     {
      message=Chillllllll.
      contextRequirements [{requirement {
         class=FortControllerRequirement_HasCID
         properties {allowedCIds [/BRCosmetics/Athena/Items/Cosmetics/Characters/CID_798_Athena_Commando_M_JonesyVagabond]}
        }}]
      priority {value=0, curve {curveTable=null, rowName=None}}
     }
     {
      message="You know, I was kind of a goth back in the day."
      contextRequirements [{requirement {class=FortControllerRequirement_HasCID, properties {allowedCIds [&Character_ChillCat]}}}]
      priority {value=0, curve {curveTable=null, rowName=None}}
     }
     {
      message="You should get those hands checked out, my guy."
      contextRequirements [{requirement {class=FortControllerRequirement_HasCID, properties {allowedCIds [&Character_Bites]}}}]
      priority {value=0, curve {curveTable=null, rowName=None}}
     }
     {
      message="Season 2, episode 11. Changed television forever."
      contextRequirements [{requirement {class=FortControllerRequirement_HasCID, properties {allowedCIds [&Character_PinkSpike]}}}]
      priority {value=0, curve {curveTable=null, rowName=None}}
     }
     {
      message="Aw man, you're just a little guy, huh?"
      contextRequirements [{requirement {class=FortControllerRequirement_HasCID, properties {allowedCIds [&Character_BadBear]}}}]
      priority {value=0, curve {curveTable=null, rowName=None}}
     }
     {
      message="Oh,  yeah. I'm supposed to be watching you."
      contextRequirements [{requirement {class=FortControllerRequirement_HasCID, properties {allowedCIds [&Character_TheHerald]}}}]
      priority {value=0, curve {curveTable=null, rowName=None}}
     }
     {
      message="You're not from here, are you?"
      contextRequirements [{requirement {class=FortControllerRequirement_HasCID, properties {allowedCIds [&Character_RoseDust]}}}]
      priority {value=0, curve {curveTable=null, rowName=None}}
     }
    ]
   }
  }
 }
 ConversationLinkNode_0 {id=#4, class=ConversationLinkNode, properties {remoteEntryTag {tagName=Conversation.Entry.BasicServices}}}
 ConversationLinkNode_1 {id=#3, class=ConversationLinkNode, properties {remoteEntryTag {tagName=Conversation.Entry.NPCLibrary.Quests}}}
]
